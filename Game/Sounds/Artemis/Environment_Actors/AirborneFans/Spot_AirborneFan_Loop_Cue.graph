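# Looping spot sound cue for an airborne fan, routed to the gadgets sound class.
# Two looping layers, close fan blades and wind, each pass through their own attenuation asset before being mixed at equal volume.
# The same mix feeds every quality-level input.
# The cue also carries cone-shaped attenuation overrides with logarithmic falloff, occlusion and listener focus, though the override flag is not set.
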SoundCue /Game/Sounds/Artemis/Environment_Actors/AirborneFans/Spot_AirborneFan_Loop_Cue
root=QualityLevel_0
nodes [
 {
  id=Attenuation_0
  children [WavePlayer_0]
  properties {attenuationSettings=&/Spot_AirborneFan_Blades_Close_Att}
 }
 {
  id=Attenuation_1
  children [WavePlayer_3]
  properties {attenuationSettings=&/Spot_AirborneFan_Air_Att}
 }
 {id=Mixer_0, children [Attenuation_0, Attenuation_1], properties {inputVolume [1,1]}}
 {id=QualityLevel_0, children [Mixer_0,Mixer_0,Mixer_0,Mixer_0]}
 {
  id=WavePlayer_0
  properties {bLooping=true, soundWaveAssetPtr=&/Spot_AirborneFan_Blades_Close_Loop}
 }
 {
  id=WavePlayer_3
  properties {bLooping=true, soundWaveAssetPtr=&/Spot_AirborneFan_Wind_Loop}
 }
]
properties {
 bDebug=true
 maxDistance=4000
 soundClassObject=&Fort_Gadgets
 volumeMultiplier=0.66
 attenuationOverrides {
  attenuationShape=Cone
  bEnableFocusInterpolation=true
  bEnableListenerFocus=true
  bEnableOcclusion=true
  coneSphereRadius=600
  distanceAlgorithm=Logarithmic
  falloffDistance=1000
  nonFocusVolumeAttenuation=0.7
  nonSpatializedRadiusStart=1500
  occlusionLowPassFilterFrequency=10000
  occlusionTraceChannel=ECC_GameTraceChannel11
  occlusionVolumeAttenuation=0.5
  stereoSpread=10000
  attenuationShapeExtents=<7500 8 10>
 }
}
