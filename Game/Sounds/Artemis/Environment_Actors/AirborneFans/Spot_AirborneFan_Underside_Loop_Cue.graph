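# Looping spot cue for the underside of airborne fans, reusing a blimp vortex downdraft wind wave.
# A quality-level root feeds one looping wave player to its first two levels and leaves the other two empty.
# It plays in the gadgets sound class with the blimp downdraft cone attenuation asset.
# Cone, focus and occlusion override values are stored but not shown as enabled, and the debug flag is left on.
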SoundCue /Game/Sounds/Artemis/Environment_Actors/AirborneFans/Spot_AirborneFan_Underside_Loop_Cue
root=QualityLevel_0
nodes [
 {id=QualityLevel_0, children [WavePlayer_1,WavePlayer_1,null,null]}
 {id=WavePlayer_1, properties {bLooping=true, soundWaveAssetPtr=/Game/Sounds/Artemis/POI/Blimps/Spots/Blimp_Vortex_Wind_Downdraft_Loop_01}}
]
properties {
 attenuationSettings=/Game/Sounds/Artemis/POI/Blimps/Spots/Blimp_Vortex_Wind_Downdraft_Cone_Att
 bDebug=true
 maxDistance=2400
 soundClassObject=&Fort_Gadgets
 volumeMultiplier=0.66
 attenuationOverrides {
  attenuationShape=Cone
  bEnableFocusInterpolation=true
  bEnableListenerFocus=true
  bEnableOcclusion=true
  coneSphereRadius=600
  distanceAlgorithm=Logarithmic
  falloffDistance=1000
  nonFocusVolumeAttenuation=0.7
  nonSpatializedRadiusStart=1500
  occlusionLowPassFilterFrequency=10000
  occlusionTraceChannel=ECC_GameTraceChannel11
  occlusionVolumeAttenuation=0.5
  stereoSpread=10000
  attenuationShapeExtents=<7500 8 10>
 }
}
